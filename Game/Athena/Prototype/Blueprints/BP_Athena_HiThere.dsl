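; BP_Athena_HiThere is a prototype Athena Blueprint that reports the content patch version when play begins, apparently to test emergency content patching.
; On BeginPlay, only on non-server instances, it adds a "Version" attribute holding the ContentPatchVersion variable and fires the "EmergencyContentPatchTest" analytics event.
; It also builds a version message string, passes it to PrintString and writes it through BuildingFunctionLibrary's LogString.
; Overlap, Tick and the construction script are empty.
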(event EventBeginPlay
  (bind _contentpatchversion (Variables|Default|GetContentPatchVersion))
  (bind _returnvalue (Utilities|String|ToString(Integer) _contentpatchversion))
  (bind _array (Utilities|Array|MakeArray "Content patch version number is " _returnvalue))
  (bind _returnvalue_2 (Utilities|String|JoinStringArray _array))
  (bind _attributearray (Variables|Default|GetAttributeArray))
  (if (not (Networking|IsServer))
    (bind _returnvalue_3 (Analytics|ConstructAnalyticsAttribute :AttributeName "Version" :IntValue _contentpatchversion :floatValue -1.0 :BoolValue false))
    (Utilities|Array|Add _attributearray _returnvalue_3)
    (Analytics|FireAnalyticsEventBlueprintArray "EmergencyContentPatchTest" _attributearray)
    (Development|PrintString _returnvalue_2 false)
    (Class|BuildingFunctionLibrary|LogString "NewEnumerator3" _returnvalue_2)))

(event Collision|EventActorBeginOverlap
  (params (OtherActor :type Actor%20Object%20Reference)))

(event EventTick
  (params (DeltaSeconds :type Float%20%28single-precision%29)))

(fn ConstructionScript)
