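# Prototype translucent, unlit, two-sided laser beam material, flagged for Niagara mesh particles and responsive AA.
# SplineThicken takes a base and tip width from pixel depth divided by view width times Width, scaling thickness with distance to hold on-screen size.
# A color parameter times Multiplier gives the glow color.
# A squared center-peaked falloff on U, multiplied by a power of U and Opacity, shapes the beam fade.
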Material /Game/Athena/Prototype/M_LaserBeam
graph {
 properties {
  bEnableResponsiveAA=true
  bUsedWithNanite=true
  bUsedWithNiagaraMeshParticles=true
  blendMode=BLEND_Translucent
  pixelDepthOffsetMode=PDOM_Legacy
  shadingModel=MSM_Unlit
  twoSided=true
  usedShadingModels=Unlit
 }
 nodes [
  n1 = ViewSize()
  n2 = Divide(A=PixelDepth().R, B=Multiply(A=ComponentMask(None=@n1) {r=true}, B=ScalarParameter() {defaultValue=0.25, parameterName=Width}))
  MaterialFunctionCall(WidthBase=@n2, WidthTip=@n2) {materialFunction=&SplineThicken}
  Multiply(A=VectorParameter() {defaultValue=<rgba 20 0 0 1>}.RGB, B=ScalarParameter() {defaultValue=1, parameterName=Multiplier})
  n3 = TextureCoordinate()
  n4 = ComponentMask(None=@n3) {r=true}
  n5 = Constant() {r=0.5}
  n6 = OneMinus(None=Multiply(A=Distance(A=@n4, B=@n5)) {constB=2})
  n7 = ComponentMask(None=@n3) {r=true}
  Multiply(A=Multiply(A=Multiply(A=@n6, B=@n6), B=Power(Base=@n7)), B=ScalarParameter() {defaultValue=1, parameterName=Opacity})
  OneMinus(None=Clamp(None=@n2))
  ScalarParameter() {defaultValue=0.1}
 ]
}
